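# Material function for the female green jacket glider FX that computes a world position offset to add movement to a mesh particle.
# It subtracts particle world position from a vector built from one dynamic parameter channel, then symmetrically clamps the result.
# The clamp limit blends between a scalar parameter and a dynamic parameter, noted as a fallback for diving issues.
# An inverted-U gradient with adjustable power, times dynamic alpha, scales the bend.
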MaterialFunction /Game/Gadgets/Assets/VinderTech_GliderChute/Glider_Green_Jacket_Female/FX/MF_WPO_Interpolation
graph {
 properties {description=Add World Movement to a Static Mesh, previewBlendMode=BLEND_Masked}
 nodes [
  n1 = ParticlePositionWS()
  ComponentMask(None=@n1) {b=true}
  n2 = DynamicParameter() {desc=Dyn1 WPO Feed, parameterIndex=1, defaultValue=<rgba 1 1 50 1>, paramNames [WPO UV Scale, "", "", ""]}
  n3 = Subtract(
   A=MaterialFunctionCall(X=@n2.WPO UV Scale, Y=@n2.WPO UV Scale, Z=@n2.WPO UV Scale) {materialFunction=&MakeFloat3}.Result
   B=@n1
  )
  n4 = ScalarParameter() {defaultValue=10, group="☄WPO", parameterName=Interpolate Clamp}
  n5 = Clamp(None=@n3, Min=Multiply(B=@n4) {constA=-1}, Max=@n4) {maxDefault=1000, minDefault=-1000}
  n6 = DynamicParameter() {desc=Dyn0, defaultValue=<rgba 1 1 1 1>, paramNames [ClampInterpolate, WingOffset_01, WingOffset_02, WPO_Wave_Direction]}
  n7 = Clamp(None=@n3, Min=Multiply(B=@n6.ClampInterpolate) {constA=-1}, Max=@n6.ClampInterpolate) {maxDefault=1000, minDefault=-1000}
  n8 = ScalarParameter() {defaultValue=1, group="☄WPO", parameterName=UseDynamicParameterClampInterpolate}
  n9 = TextureCoordinate()
  n10 = Power(Base=ComponentMask(None=OneMinus(None=@n9)) {r=true}, Exp=ScalarParameter() {defaultValue=2, group="☄WPO", parameterName=WPOBendPower})
  FunctionOutput(None=LinearInterpolate(B=LinearInterpolate(A=@n5, B=@n7, Alpha=@n8) {desc=In case Diving causes issues}, Alpha=Multiply(A=@n10, B=@n2.A)))
 ]
}
