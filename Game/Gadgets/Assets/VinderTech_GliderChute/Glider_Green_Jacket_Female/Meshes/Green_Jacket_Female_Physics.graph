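# Physics asset for the female Green Jacket variant of the VinderTech GliderChute, defining simulation bodies for its root and two wings.
# Each bone has a single capsule, with larger, offset and rotated capsules on the wing tips.
# Constraints form two chains from Root through each wing to its tip.
# Wing-to-root joints swing and twist freely, while tip-to-wing joints set angular limits with a tighter twist range.
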PhysicsAsset /Game/Gadgets/Assets/VinderTech_GliderChute/Glider_Green_Jacket_Female/Meshes/Green_Jacket_Female_Physics
bodies {
 Root [{shapeType=Capsule, length=3, radius=3}]
 Wing_l [{shapeType=Capsule, length=3, radius=3}]
 Wing_l_tip [{shapeType=Capsule, length=9.7444, radius=4.1583, center=<-9.2928 -0.3066 0.0924>, rotation=<rot 0 -85.9485 -89.4817>}]
 Wing_r [{shapeType=Capsule, length=3, radius=3}]
 Wing_r_tip [{shapeType=Capsule, length=9.705, radius=4.1589, center=<9.3311 0.299 -0.0103>, rotation=<rot 0 93.9807 -89.9872>}]
}
constraints [
 {
  bone1Name=Wing_l
  bone2Name=Root
  swing1LimitDegrees=45
  swing1Motion=Free
  swing2LimitDegrees=45
  swing2Motion=Free
  twistLimitDegrees=45
  twistMotion=Free
 }
 {
  bone1Name=Wing_r
  bone2Name=Root
  swing1LimitDegrees=45
  swing1Motion=Free
  swing2LimitDegrees=45
  swing2Motion=Free
  twistLimitDegrees=45
  twistMotion=Free
 }
 {bone1Name=Wing_l_tip, bone2Name=Wing_l, swing1LimitDegrees=45, swing2LimitDegrees=45, twistLimitDegrees=25}
 {bone1Name=Wing_r_tip, bone2Name=Wing_r, swing1LimitDegrees=45, swing2LimitDegrees=45, twistLimitDegrees=25}
]
